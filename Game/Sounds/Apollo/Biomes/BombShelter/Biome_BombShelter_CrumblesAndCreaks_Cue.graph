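# Bomb Shelter biome ambient cue that plays one-shot crumble and creak sounds in the indoor ambience sound class.
# An equally weighted Random node picks one of eight waves, and a Modulator varies its pitch and volume.
# The root quality switch feeds that chain to its first two levels and leaves the other two empty.
# A tile-ambient one-shot attenuation asset handles falloff, and a concurrency override caps simultaneous instances.
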SoundCue /Game/Sounds/Apollo/Biomes/BombShelter/Biome_BombShelter_CrumblesAndCreaks_Cue
root=QualityLevel_0
nodes [
 {id=Modulator_1, children [Random_1], properties {pitchMin=0.8, volumeMax=1, volumeMin=0.8}}
 {id=QualityLevel_0, children [Modulator_1,Modulator_1,null,null]}
 {
  id=Random_1
  children [WavePlayer_0, WavePlayer_7, WavePlayer_8, WavePlayer_9, WavePlayer_10, WavePlayer_11, WavePlayer_15, WavePlayer_16]
  properties {weights [1,1,1,1,1,1,1,1]}
 }
 {id=WavePlayer_0, properties {soundWaveAssetPtr=&/Amb_BombShelter_CrumblesAndCreaks_01}}
 {id=WavePlayer_10, properties {soundWaveAssetPtr=&/Amb_BombShelter_CrumblesAndCreaks_05}}
 {id=WavePlayer_11, properties {soundWaveAssetPtr=&/Amb_BombShelter_CrumblesAndCreaks_06}}
 {id=WavePlayer_15, properties {soundWaveAssetPtr=&/Amb_BombShelter_CrumblesAndCreaks_07}}
 {id=WavePlayer_16, properties {soundWaveAssetPtr=&/Amb_BombShelter_CrumblesAndCreaks_08}}
 {id=WavePlayer_7, properties {soundWaveAssetPtr=&/Amb_BombShelter_CrumblesAndCreaks_02}}
 {id=WavePlayer_8, properties {soundWaveAssetPtr=&/Amb_BombShelter_CrumblesAndCreaks_03}}
 {id=WavePlayer_9, properties {soundWaveAssetPtr=&/Amb_BombShelter_CrumblesAndCreaks_04}}
]
properties {
 attenuationSettings=/Game/Sounds/Attenuation/Fort_TileAmbient_OneShot_Att
 bOverrideConcurrency=true
 maxDistance=8000
 soundClassObject=&Fort_Ambient_Indoors_Athena
 volumeMultiplier=0.351
 concurrencyOverrides {maxCount=7}
 preEffectBusSends [{}]
}
